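; Gameplay ability that monitors a hired (converted) NPC's progress toward upgrading, based on its converter pawn's damage or a tagged gameplay event.
; Each threshold or event raises a level and plays a cue.
; At the final level it optionally swaps cosmetics and removes a blocking tag.
; It can then wait for Unaware alert level or timer-polled converter proximity before calling OnFullUpgrade, an empty hook here.
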(event Ability|EventActivateAbility
  (Ability|CommitAbility)
  (Utilities|FlowControl|Delay 1.0)
  (bind _outfortplayerpawn (CallFunction|GetAvatarasValidFortPlayerPawn)
    (:ValidFortPlayerPawn
      (bind _output_get (Variables|InternalData|SetAIPawn _outfortplayerpawn))
      (bind _output_get_2 (Variables|InternalData|SetConvertComponent (Actor|GetComponentbyClass _output_get "/Script/FortniteGame.FortPawnComponent_Convert")))
      (Utilities|IsValid _output_get_2
        (:Is%20Valid
          (bind _asfort_athena_aibot_controller (Utilities|Casting|CastToFortAthenaAIBotController (Pawn|GetController (Variables|InternalData|GetAIPawn)))
            (:then
              (Variables|InternalData|SetControllerRef _asfort_athena_aibot_controller)
              (Variables|InternalData|SetConverterPawn (Convert|GetConverterPawn (Variables|InternalData|GetConvertComponent)))
              (Utilities|FlowControl|Sequence
                (:then_0
                  (if (Variables|Default|GetWaitForUnaware)
                    (AlertLevel|BindEventtoOnAlertLevelChangedEvent (Variables|InternalData|GetControllerRef) (EventDispatchers|CreateEvent :!event Custom|AlertLevelChanged))))
                (:then_1
                  (switch Utilities|FlowControl|Switch|SwitchonENUM_UpgradeSpecialist_UpgradeType (Variables|Default|GetUpgradeType)
                    (:NewEnumerator0
                      (EventDispatchers|BindEventtoOnDidDamage (Variables|InternalData|GetConverterPawn) (EventDispatchers|CreateEvent :!event Custom|ConverterCausedDamage)))
                    (:NewEnumerator1
                      (Ability|Async|WaitGameplayEventToActor (select (ScalableFloat|GetValueasBool (Variables|Default|GameplayEvent|GetListenForEventOnHiredPawn)) (Variables|InternalData|GetAIPawn) (Variables|InternalData|GetConverterPawn)) (Variables|Default|GameplayEvent|GetGameplayEventTagToListenTo)
                        (:EventReceived
                          (CallFunction|TriggerUpgrade))))))))
            (:CastFailed
              _)))
        (:Is%20Not%20Valid
          _))
      (Ability|EndAbility))
    (:AvatarCastFailed
      _)))

(event Ability|EventOnEndAbility
  (params (bWasCancelled :type Boolean))
  (bind _converterpawn (Variables|InternalData|GetConverterPawn))
  (bind _controllerref (Variables|InternalData|GetControllerRef))
  (Utilities|Time|ClearandInvalidateTimerbyHandle (Variables|Default|GetCheckProximity_TH))
  (Utilities|IsValid _converterpawn
    (:Is%20Valid
      (EventDispatchers|UnbindEventfromOnDidDamage _converterpawn (EventDispatchers|CreateEvent :!event Custom|ConverterCausedDamage))
      (Utilities|IsValid _controllerref
        (:Is%20Valid
          (AlertLevel|UnbindEventfromOnAlertLevelChangedEvent _controllerref (EventDispatchers|CreateEvent :!event Custom|AlertLevelChanged)))))))

(event Custom|ConverterCausedDamage
  (params (DamagedActor :type Actor%20Object%20Reference) (Damage :type Float%20%28single-precision%29) (InstigatedBy :type Controller%20Object%20Reference) (DamageCauser :type Actor%20Object%20Reference) (HitLocation :type Vector) (FHitComponent :type Primitive%20Component%20Object%20Reference) (BoneName :type Name) (Momentum :type Vector))
  ; Listen for Damage Caused
  (bind _output_get (Variables|InternalData|SetCurrentDamageDealt (+ (Variables|InternalData|GetCurrentDamageDealt) Damage)))
  (if (>= _output_get (ScalableFloat|GetValueatLevel (Variables|Default|Damage|GetDamageToRaiseDing)))
    (CallFunction|TriggerUpgrade)))

(event Custom|CheckProximityToConverter
  (if (and (<= (Transformation|GetDistanceTo (Variables|InternalData|GetAIPawn) (Variables|InternalData|GetConverterPawn)) (ScalableFloat|GetValueatLevel (Variables|Default|GetConverterProximityRange_Hf))) (select (Variables|Default|GetWaitForUnaware) (Utilities|Enum|Equal(Enum) (Class|FortAthenaAIBotController|GetAlertLevel (Variables|InternalData|GetControllerRef)) "Unaware") true))
    (Utilities|Time|ClearandInvalidateTimerbyHandle (Variables|Default|GetCheckProximity_TH))
    (CallFunction|OnFullUpgrade)))

(event Custom|TriggerUpgrade
  (bind _tagpreventability (Variables|Default|GetTagPreventAbility))
  (bind _aipawn (Variables|InternalData|GetAIPawn))
  (bind _aipawn_2 (Variables|InternalData|GetAIPawn))
  (Math|Integer|IncrementInt (Variables|InternalData|GetCurrentLevel)
    (:%20%20
      (Variables|InternalData|SetCurrentDamageDealt 0.0)
      (Ability|ExecuteGameplayCueOnOwner (Variables|Default|GetGC_UpgradeDing))
      (if (>= (Variables|InternalData|GetCurrentLevel) (ScalableFloat|GetValueatLevel (Variables|Default|GetNumberOfLevelsForFullUpgrade)))
        (Utilities|FlowControl|DoOnce false
          (:Completed
            (if (ScalableFloat|GetValueasBool (Variables|Default|Cosmetics|GetShouldUpdateCosmeticOnFullUpgrade))
              (bind _returnvalue (GUID|GenerateGuidfromObject _aipawn)) ; Needed for Swap data
              (Class|FortPlayerPawn|SwapPawnCosmeticItemandVariant _aipawn _returnvalue (Variables|Default|Cosmetics|GetNewCharacterData) false "Active" false) ; Change the outfit out to the night version
              (else
                _))
            (Ability|GameplayEffect|RemoveGameplayTags _aipawn_2 _tagpreventability)
            (Ability|GameplayEffect|RemoveLooseGameplayTags _aipawn_2 _tagpreventability true)
            (if (Variables|Default|GetWaitForUnaware)
              (if (not (Utilities|Enum|Equal(Enum) (Class|FortAthenaAIBotController|GetAlertLevel (Variables|InternalData|GetControllerRef)) "Unaware"))
                (Variables|InternalData|SetReadyToUpgrade true)))
            (CallFunction|AttemptFullUpgrade)))))))

(event Custom|AlertLevelChanged
  (params (BotController :type Fort%20Athena%20AIBot%20Controller%20Object%20Reference) (OldAlertLevel :type EAlertLevel%20Enum :default "Unaware") (NewAlertLevel :type EAlertLevel%20Enum :default "Unaware"))
  (switch Utilities|FlowControl|Switch|SwitchonEAlertLevel NewAlertLevel
    (:Unaware
      (if (Variables|InternalData|GetReadyToUpgrade)
        (CallFunction|AttemptFullUpgrade)))))

(event Custom|AttemptFullUpgrade
  (if (Variables|Default|GetReachConverterFirst)
    (bind _returnvalue (Utilities|Time|SetTimerbyEvent (delegate Custom|CheckProximityToConverter) 1.0 true true))
    (Variables|Default|SetCheckProximity_TH _returnvalue)
    (else
      (CallFunction|OnFullUpgrade))))

(fn OnFullUpgrade)
